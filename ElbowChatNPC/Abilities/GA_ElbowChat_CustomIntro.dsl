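; GA_ElbowChat_CustomIntro is a gameplay ability that plays an intro montage for an ElbowChat NPC.
; On activation it reads the montage, play rate and root motion scale from the Fort montage info, runs parent logic, then briefly delays.
; It then pushes behavior control on the AI controller, commits the ability and starts the montage task.
; The end event only forwards to the parent, and no matching behavior-control pop appears here.
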(event Ability|EventActivateAbility
  (bind (_montagetoplay _animplayrate _animrootmotiontranslationscale) (Utilities|Struct|BreakFortGameplayAbilityMontageInfo (Variables|Default|GetFort%20Gameplay%20Ability%20Montage%20Info)))
  (|CallParentFunction)
  (Utilities|FlowControl|Delay 0.2)
  (AI|PushBehaviorControl (Variables|InternalData|GetAIController) -2088960)
  (Ability|CommitAbility)
  (Ability|Tasks|PlayMontageAndWaitNotify "None" _montagetoplay _animplayrate "None" false _animrootmotiontranslationscale))

(event Ability|EventFailedToActivatePassiveAbility
  (params (ActorInfo :type Gameplay%20Ability%20Actor%20Info%20Structure)))

(event Ability|EventOnEndAbility
  (params (bWasCancelled :type Boolean))
  (|CallParentFunction bWasCancelled))
